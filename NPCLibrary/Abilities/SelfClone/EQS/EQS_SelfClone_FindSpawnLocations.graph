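# Environment query that finds spawn locations around the querier for the NPC SelfClone ability.
# A spiral donut generator builds rings of candidate points around the querier and projects them to navigation with a no-smashable, no-water filter, raised slightly afterward.
# A pathfinding test using the any-traversable nav filter and an obstruction test against the querier each filter and score the points.
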EnvQuery /NPCLibrary/Abilities/SelfClone/EQS/EQS_SelfClone_FindSpawnLocations
options [{
  generator {
   name=EnvQueryGenerator_Donut_0
   class=EnvQueryGenerator_Donut
   properties {
    bUseSpiralPattern=true
    center=&EnvQueryContext_Querier
    innerRadius=550
    numberOfRings=3
    outerRadius=1000
    pointsPerRing=15
    projectionData {
     extentX=0
     navigationFilter=&NavFilter_Phoebe_NoSmashableNoWater_C
     postProjectionVerticalOffset=15
    }
   }
  }
  tests [
   EnvQueryTest_Pathfinding_0 {
    class=EnvQueryTest_Pathfinding
    properties {filterClass=&NavFilter_NPC_AnyTraversable_C, testPurpose=FilterAndScore}
   }
   FortQueryTest_IsObstructed_0 {class=FortQueryTest_IsObstructed, properties {context=&EnvQueryContext_Querier, testPurpose=FilterAndScore}}
  ]
 }]
